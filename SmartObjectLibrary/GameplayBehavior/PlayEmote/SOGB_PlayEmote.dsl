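; Smart Object gameplay behavior that makes a Fortnite AI bot emote for a configured total duration.
; Init validates config, pawn and bot controller; PlayEmote picks the first, random or pawn-tag-matched emote, with optional per-emote time limits.
; The emote-stopped delegate replays looping emotes or ends the behavior.
; Optional looping speech bubbles (component injected if missing) and focus along the Smart Object's forward are cleaned up on finish.
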(event GameplayBehavior|EventOnTriggeredCharacter
  (params (Avatar :type Character%20Object%20Reference) (Config :type Gameplay%20Behavior%20Config%20Object%20Reference) (SmartObjectOwner :type Actor%20Object%20Reference))
  (bind _playemoteconfig (Variables|Default|GetPlayEmoteConfig))
  (bind _startdelay (Variables|/SmartObjectLibrary/GameplayBehavior/PlayEmote/SOGB_PlayEmote_Config.SOGB_PlayEmote_Config_C|GetStartDelay _playemoteconfig))
  ;[1 1 1] On Start Behavior
  (bind _isvalid (CallFunction|Init Avatar Config SmartObjectOwner))
  (if _isvalid
    (CallFunction|HandleSpeechBubbleComponent)
    (if (Variables|/SmartObjectLibrary/GameplayBehavior/PlayEmote/SOGB_PlayEmote_Config.SOGB_PlayEmote_Config_C|GetForceFocalPointOnActor (Variables|Default|GetPlayEmoteConfig))
      (CallFunction|SetPawnFocalPoint false)
      (else
        _))
    (if (> _startdelay 0)
      (GameplayTasks|TaskWaitDelay 0 _startdelay
        (:OnFinish
          _))
      (else
        _))
    (GameplayTasks|TaskWaitDelay 0 (Variables|/SmartObjectLibrary/GameplayBehavior/PlayEmote/SOGB_PlayEmote_Config.SOGB_PlayEmote_Config_C|GetEmotingTotalDuration _playemoteconfig)
      (:then
        (CallFunction|PlayEmote)
        (CallFunction|StartSpeechBubbles)
        (CallFunction|EmoteStarted))
      (:OnFinish
        (GameplayBehavior|EndBehavior (Variables|Default|GetUserAvatar))))
    (else
      (GameplayBehavior|EndBehavior Avatar))))

(event Custom|StartSpeechBubbles
  (bind _userpawn (Variables|InternalData|GetUserPawn))
  (bind _returnvalue (Actor|GetComponentbyClass _userpawn "/Script/FortniteGame.FortPawnComponent_SpeechBubble"))
  (bind _playemoteconfig (Variables|Default|GetPlayEmoteConfig))
  (bind _speechbubbledelay (Variables|/SmartObjectLibrary/GameplayBehavior/PlayEmote/SOGB_PlayEmote_Config.SOGB_PlayEmote_Config_C|GetSpeechBubbleDelay _playemoteconfig))
  ;[1 1 1] New Speech Bubbles Logic
  (if (Variables|/SmartObjectLibrary/GameplayBehavior/PlayEmote/SOGB_PlayEmote_Config.SOGB_PlayEmote_Config_C|GetUseSpeechBubbles _playemoteconfig)
    (Utilities|IsValid _returnvalue
      (:Is%20Valid
        (Variables|InternalData|SetSpeechBubbleComponent _returnvalue)
        (if (> _speechbubbledelay 0)
          (GameplayTasks|TaskWaitDelay 0 _speechbubbledelay
            (:OnFinish
              _))
          (else
            _))
        (CallFunction|TriggerSpeechBubble)))))

(event Custom|TriggerSpeechBubble
  (bind _playemoteconfig (Variables|Default|GetPlayEmoteConfig))
  (bind _speechbubbleinterval (Variables|/SmartObjectLibrary/GameplayBehavior/PlayEmote/SOGB_PlayEmote_Config.SOGB_PlayEmote_Config_C|GetSpeechBubbleInterval _playemoteconfig))
  (bind _playemoteconfig_2 (Variables|Default|GetPlayEmoteConfig))
  ;[1 1 1] New Speech Bubbles Logic
  (SpeechBubble|StartSpeech (Variables|InternalData|GetSpeechBubbleComponent) (Variables|/SmartObjectLibrary/GameplayBehavior/PlayEmote/SOGB_PlayEmote_Config.SOGB_PlayEmote_Config_C|GetSpeechBubbleGameplayTagQuery _playemoteconfig_2))
  (GameplayTasks|TaskWaitDelay 0 (Variables|/SmartObjectLibrary/GameplayBehavior/PlayEmote/SOGB_PlayEmote_Config.SOGB_PlayEmote_Config_C|GetSpeechBubbleDuration _playemoteconfig_2)
    (:OnFinish
      (SpeechBubble|StopSpeech (Variables|InternalData|GetSpeechBubbleComponent))
      (if (Variables|/SmartObjectLibrary/GameplayBehavior/PlayEmote/SOGB_PlayEmote_Config.SOGB_PlayEmote_Config_C|GetLoopSpeechBubbles _playemoteconfig)
        (if (> _speechbubbleinterval 0)
          (GameplayTasks|TaskWaitDelay 0 _speechbubbleinterval
            (:OnFinish
              _))
          (else
            _))
        (CallFunction|TriggerSpeechBubble)))))

(event Custom|PlayEmote
  (bind (_primaryassetid _duration) (CallFunction|GetEmoteandDuration))
  ;[1 1 1] Play Emote
  (Utilities|FlowControl|Sequence
    (:then_0
      (Utilities|FlowControl|DoOnce false
        (:Completed
          (if (Math|Boolean|ANDBoolean (Variables|/SmartObjectLibrary/GameplayBehavior/PlayEmote/SOGB_PlayEmote_Config.SOGB_PlayEmote_Config_C|GetUseRandomEmoteInList (Variables|Default|GetPlayEmoteConfig)) (Variables|/SmartObjectLibrary/GameplayBehavior/PlayEmote/SOGB_PlayEmote_Config.SOGB_PlayEmote_Config_C|GetUseSameEmoteAtEndWhenLooping (Variables|Default|GetPlayEmoteConfig)) (not (Variables|/SmartObjectLibrary/GameplayBehavior/PlayEmote/SOGB_PlayEmote_Config.SOGB_PlayEmote_Config_C|GetUseEmoteTags (Variables|Default|GetPlayEmoteConfig))))
            (Variables|InternalData|SetPersistedSelectedIndex (CallFunction|GetRandomEmoteIndexfromList))))))
    (:then_1
      (if (not (Variables|/SmartObjectLibrary/GameplayBehavior/PlayEmote/SOGB_PlayEmote_Config.SOGB_PlayEmote_Config_C|GetUseEmoteTags (Variables|Default|GetPlayEmoteConfig)))
        (Variables|InternalData|SetSelectedIndex (select (Variables|/SmartObjectLibrary/GameplayBehavior/PlayEmote/SOGB_PlayEmote_Config.SOGB_PlayEmote_Config_C|GetUseRandomEmoteInList (Variables|Default|GetPlayEmoteConfig)) (select (Variables|/SmartObjectLibrary/GameplayBehavior/PlayEmote/SOGB_PlayEmote_Config.SOGB_PlayEmote_Config_C|GetUseSameEmoteAtEndWhenLooping (Variables|Default|GetPlayEmoteConfig)) (Variables|InternalData|GetPersistedSelectedIndex) (select (Variables|/SmartObjectLibrary/GameplayBehavior/PlayEmote/SOGB_PlayEmote_Config.SOGB_PlayEmote_Config_C|GetUseDifferentRandomEmote (Variables|Default|GetPlayEmoteConfig)) (CallFunction|GetDifferentRandomEmoteIndexfromList (Variables|InternalData|GetSelectedIndex)) (CallFunction|GetRandomEmoteIndexfromList))) 0)))
      (Class|FortAthenaAIBotController|PlayEmote (Variables|InternalData|GetUserController) _primaryassetid)
      (Utilities|FlowControl|Sequence
        (:then_0
          (Utilities|FlowControl|DoOnce false
            (:Completed
              (EventDispatchers|BindEventtoOnPawnStoppedEmote (Variables|InternalData|GetUserPawn) (delegate Custom|OnStopEmoteEvent)))))
        (:then_1
          (if (> _duration 0)
            (bind _waitasynctask (Variables|InternalData|GetWaitAsyncTask)
              (:then
                (GameplayTasks|EndTask _waitasynctask))
              (:else
                _))
            (bind _asynctaskproxy (GameplayTasks|TaskWaitDelay 0 _duration)
              (:then
                (Variables|InternalData|SetWaitAsyncTask _asynctaskproxy))
              (:OnFinish
                (Class|FortAthenaAIBotController|StopEmote (Variables|InternalData|GetUserController))))))))))

(event Custom|OnStopEmoteEvent
  (params (MontageItemDef :type Fort%20Item%20Definition%20Object%20Reference) (PawnEmoting :type Fort%20Pawn%20Object%20Reference))
  (bind _playemoteconfig (Variables|Default|GetPlayEmoteConfig))
  (bind _emoteinterval (Variables|/SmartObjectLibrary/GameplayBehavior/PlayEmote/SOGB_PlayEmote_Config.SOGB_PlayEmote_Config_C|GetEmoteInterval _playemoteconfig))
  ;[1 1 1] On Stop Emote
  (if (Variables|/SmartObjectLibrary/GameplayBehavior/PlayEmote/SOGB_PlayEmote_Config.SOGB_PlayEmote_Config_C|GetLoopEmote (Variables|Default|GetPlayEmoteConfig))
    (if (> _emoteinterval 0.1)
      (GameplayTasks|TaskWaitDelay 0 _emoteinterval
        (:OnFinish
          _))
      (else
        (GameplayTasks|TaskWaitDelay 0 0.1
          (:OnFinish
            _))))
    (CallFunction|PlayEmote)
    (else
      (GameplayBehavior|EndBehavior (Variables|Default|GetUserAvatar)))))

(event GameplayBehavior|EventOnFinishedCharacter
  (params (Avatar :type Character%20Object%20Reference) (bWasInterrupted :type Boolean))
  (bind _userpawn (Variables|InternalData|GetUserPawn))
  (bind _returnvalue (Actor|GetComponentbyClass _userpawn "/Script/FortniteGame.FortPawnComponent_SpeechBubble"))
  (bind _userpawn_2 (Variables|InternalData|GetUserPawn))
  (bind _usercontroller (Variables|InternalData|GetUserController))
  ;[1 1 1] On Stop Behavior
  (Utilities|IsValid _userpawn_2
    (:Is%20Valid
      (EventDispatchers|UnbindEventfromOnPawnStoppedEmote _userpawn_2 (delegate Custom|OnStopEmoteEvent))
      (Utilities|IsValid _usercontroller
        (:Is%20Valid
          (Class|FortAthenaAIBotController|StopEmote _usercontroller)
          (CallFunction|EmoteEnded)
          (if (Variables|/SmartObjectLibrary/GameplayBehavior/PlayEmote/SOGB_PlayEmote_Config.SOGB_PlayEmote_Config_C|GetUseSpeechBubbles (Variables|Default|GetPlayEmoteConfig))
            (Utilities|IsValid _returnvalue
              (:Is%20Valid
                (SpeechBubble|StopSpeech _returnvalue))
              (:Is%20Not%20Valid
                _))
            (if (Variables|/SmartObjectLibrary/GameplayBehavior/PlayEmote/SOGB_PlayEmote_Config.SOGB_PlayEmote_Config_C|GetForceFocalPointOnActor (Variables|Default|GetPlayEmoteConfig))
              (CallFunction|SetPawnFocalPoint true))
            (else
              _)))))))

(event Custom|EmoteStarted)

(event Custom|EmoteEnded)

(fn GetDifferentRandomEmoteIndexFromList
  (params (LastEmoteIndex :type Integer))
  (bind _playemoteconfig (Variables|Default|GetPlayEmoteConfig))
  (bind _emoteandtaglist (Variables|/SmartObjectLibrary/GameplayBehavior/PlayEmote/SOGB_PlayEmote_Config.SOGB_PlayEmote_Config_C|GetEmoteAndTagList _playemoteconfig))
  (bind _returnvalue (Utilities|Array|Length _emoteandtaglist))
  (bind _returnvalue_2 (- _returnvalue 1))
  (bind _lastemoteindex (Variables|GetLastEmoteIndex))
  (if (== (Utilities|Array|Length (Variables|/SmartObjectLibrary/GameplayBehavior/PlayEmote/SOGB_PlayEmote_Config.SOGB_PlayEmote_Config_C|GetEmoteAndTagList (Variables|Default|GetPlayEmoteConfig))) 1)
    (return _lastemoteindex)
    (else
      (bind _output_get (LocalVariables|SetLastIndex _lastemoteindex))
      (if (== _output_get 0)
        (return (Math|Random|RandomIntegerinRange 1 (- (Utilities|Array|Length (Variables|/SmartObjectLibrary/GameplayBehavior/PlayEmote/SOGB_PlayEmote_Config.SOGB_PlayEmote_Config_C|GetEmoteAndTagList (Variables|Default|GetPlayEmoteConfig))) 1)))
        (elif (== (LocalVariables|GetLastIndex) _returnvalue_2)
          (return (Math|Random|RandomIntegerinRange 0 (- _returnvalue_2 1)))
          (else
            (return (select (Math|Random|RandomBool) (Math|Random|RandomIntegerinRange (+ (LocalVariables|GetLastIndex) 1) (- (Utilities|Array|Length (Variables|/SmartObjectLibrary/GameplayBehavior/PlayEmote/SOGB_PlayEmote_Config.SOGB_PlayEmote_Config_C|GetEmoteAndTagList (Variables|Default|GetPlayEmoteConfig))) 1)) (Math|Random|RandomIntegerinRange 0 (- (LocalVariables|GetLastIndex) 1))))))))))

(fn GetEmoteAndDuration
  (bind (_emote_2_572e54cd4fe3b97e2d319db8ec9534f3 _gameplaytag_5_58f1982d475aa45198c1c688f5ce251a _maxduration_8_3b73b4d3461d07a8091961a1a2973dfa) (Utilities|Struct|BreakSTRUCTPlayEmotebyTag (select (Variables|/SmartObjectLibrary/GameplayBehavior/PlayEmote/SOGB_PlayEmote_Config.SOGB_PlayEmote_Config_C|GetUseEmoteTags (Variables|Default|GetPlayEmoteConfig)) (Utilities|Array|Get(acopy) (Variables|/SmartObjectLibrary/GameplayBehavior/PlayEmote/SOGB_PlayEmote_Config.SOGB_PlayEmote_Config_C|GetEmoteAndTagList (Variables|Default|GetPlayEmoteConfig)) (Variables|InternalData|GetEmoteAndTagFoundIndex)) (Utilities|Array|Get(acopy) (Variables|/SmartObjectLibrary/GameplayBehavior/PlayEmote/SOGB_PlayEmote_Config.SOGB_PlayEmote_Config_C|GetEmoteAndTagList (Variables|Default|GetPlayEmoteConfig)) (Variables|InternalData|GetSelectedIndex)))))
  (return _emote_2_572e54cd4fe3b97e2d319db8ec9534f3 _maxduration_8_3b73b4d3461d07a8091961a1a2973dfa))

(fn GetRandomEmoteIndexFromList
  (return (Math|Random|RandomIntegerinRange 0 (- (Utilities|Array|Length (Variables|/SmartObjectLibrary/GameplayBehavior/PlayEmote/SOGB_PlayEmote_Config.SOGB_PlayEmote_Config_C|GetEmoteAndTagList (Variables|Default|GetPlayEmoteConfig))) 1))))

(fn HandleSpeechBubbleComponent
  (if (Variables|/SmartObjectLibrary/GameplayBehavior/PlayEmote/SOGB_PlayEmote_Config.SOGB_PlayEmote_Config_C|GetUseSpeechBubbles (Variables|Default|GetPlayEmoteConfig))
    (Utilities|IsValid (Actor|GetComponentbyClass (Variables|InternalData|GetUserPawn) "/Script/FortniteGame.FortPawnComponent_SpeechBubble")
      (:Is%20Valid
        (return))
      (:Is%20Not%20Valid
        (if (Utilities|IsValidClass (Variables|/SmartObjectLibrary/GameplayBehavior/PlayEmote/SOGB_PlayEmote_Config.SOGB_PlayEmote_Config_C|GetSpeechBubbleComponent (Variables|Default|GetPlayEmoteConfig)))
          (bind _returnvalue (Game|AddComponentbyClass (Variables|InternalData|GetUserPawn) 0 (Variables|/SmartObjectLibrary/GameplayBehavior/PlayEmote/SOGB_PlayEmote_Config.SOGB_PlayEmote_Config_C|GetSpeechBubbleComponent (Variables|Default|GetPlayEmoteConfig))))
          (Utilities|IsValid _returnvalue
            (:Is%20Valid
              (return))
            (:Is%20Not%20Valid
              (Development|PrintString 0 "SmartObject tried to add SpeechBubbleComponent but was not successful.")))
          (else
            (Development|PrintString 0 "Trying to use a SpeechBubble on Smart Object - but Pawn does not contain  SpeechBubbleComponent and SO is not configured to inject one.")
            (return)))))
    (else
      (return))))

(fn Init
  (params (Avatar :type Character%20Object%20Reference) (Config :type Gameplay%20Behavior%20Config%20Object%20Reference) (SmartObjectOwner :type Actor%20Object%20Reference))
  (Variables|Default|SetUserAvatar Avatar)
  (Variables|Default|SetSmartObjectOwner SmartObjectOwner)
  (bind _assogb_play_emote_config (Utilities|Casting|CastToSOGB_PlayEmote_Config Config)
    (:then
      (bind _output_get (Variables|Default|SetPlayEmoteConfig _assogb_play_emote_config))
      (if (and (> (Variables|/SmartObjectLibrary/GameplayBehavior/PlayEmote/SOGB_PlayEmote_Config.SOGB_PlayEmote_Config_C|GetEmotingTotalDuration _output_get) 0) (not (Utilities|Array|IsEmpty (Variables|/SmartObjectLibrary/GameplayBehavior/PlayEmote/SOGB_PlayEmote_Config.SOGB_PlayEmote_Config_C|GetEmoteAndTagList _output_get))))
        (bind _asfort_player_pawn (Utilities|Casting|CastToFortPlayerPawn (Variables|GetAvatar))
          (:then
            (bind _output_get_2 (Variables|InternalData|SetUserPawn _asfort_player_pawn))
            (bind _asfort_athena_aibot_controller (Utilities|Casting|CastToFortAthenaAIBotController (Pawn|GetController _output_get_2))
              (:then
                (Variables|InternalData|SetUserController _asfort_athena_aibot_controller)
                (if (Variables|/SmartObjectLibrary/GameplayBehavior/PlayEmote/SOGB_PlayEmote_Config.SOGB_PlayEmote_Config_C|GetUseEmoteTags (Variables|Default|GetPlayEmoteConfig))
                  (bind _foundemoteindex (CallFunction|ValidateTagExistinEmoteList))
                  (return (>= _foundemoteindex 0))
                  (else
                    (return true))))
              (:CastFailed
                _)))
          (:CastFailed
            _))
        (return false)
        (else
          _)))
    (:CastFailed
      _)))

(fn SetPawnFocalPoint
  (params (Reset :type Boolean))
  (if Reset
    (bind _usercontroller (Variables|InternalData|GetUserController)
      (:then
        (AI|ClearFocus _usercontroller))
      (:else))
    (else
      (bind _usercontroller_2 (Variables|InternalData|GetUserController)
        (:then
          (bind _userpawn (Variables|InternalData|GetUserPawn)
            (:then
              (AI|SetFocalPoint _usercontroller_2 (+ (Transformation|GetActorLocation _userpawn) (* (Transformation|GetActorForwardVector (Variables|Default|GetSmartObjectOwner)) 1e+03))))
            (:else)))
        (:else))))
  (return))

(fn ValidateTagExistInEmoteList
  (LocalVariables|SetFoundIndex -1)
  (bind (_array_element _array_index) (Utilities|Array|ForEachLoopwithBreak (Variables|/SmartObjectLibrary/GameplayBehavior/PlayEmote/SOGB_PlayEmote_Config.SOGB_PlayEmote_Config_C|GetEmoteAndTagList (Variables|Default|GetPlayEmoteConfig)))
    (:LoopBody
      (if (GameplayTags|HasTag (GameplayTags|GetOwnedGameplayTags (Variables|InternalData|GetUserPawn) :!class /Script/GameplayTags.GameplayTagAssetInterface) (.GameplayTag_5_58F1982D475AA45198C1C688F5CE251A (Utilities|Struct|BreakSTRUCTPlayEmotebyTag _array_element)) true)
        (bind _output_get (LocalVariables|SetFoundIndex _array_index))
        (bind _output_get_2 (Variables|InternalData|SetEmoteAndTagFoundIndex _output_get))
        (return _output_get_2)))
    (:Completed
      (return (LocalVariables|GetFoundIndex)))))
